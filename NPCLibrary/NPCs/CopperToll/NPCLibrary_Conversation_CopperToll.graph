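# Conversation database for the CopperToll NPC, Charon, a retired Underworld ferryman who chats with players in a Fortnite match.
# The entry point leads to one speech node that picks a line from contextual messages by requirement and priority.
# Dedicated lines cover the first-ever conversation and two specific player outfits, while the rest are gated by a per-match conversation count check.
# The speech node then links out to the shared NPC quest and basic services conversations.
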ConversationDatabase /NPCLibrary/NPCs/CopperToll/NPCLibrary_Conversation_CopperToll
entry_points [{tag=Conversation.Entry.CopperToll, destinations [#1]}]
nodes [
 ConversationEntryPointNode_7 {
  id=#1
  class=ConversationEntryPointNode
  outputs [#2]
  properties {entryTag {tagName=Conversation.Entry.CopperToll}}
 }
 Speech_C_4 {
  id=#2
  class=Speech_C
  outputs [#3, #4]
  properties {
   nodeName=""
   generalConfig {
    defaultMessage="New place, new me. Might take up whittling."
    contextualMessages [
     {
      message="Sorry, mate. I'm retired."
      contextRequirements [{requirement {class=FortControllerRequirement_IsFirstEverConversationWithNPC}}]
      priority {curve {rowName=Tandem.Speech.Priority.FirstEverInteraction}}
     }
     {
      message="I told you, boss. I'm retired, and I'm not coming back."
      contextRequirements [{requirement {
         class=FortControllerRequirement_HasCID
         properties {allowedCIds [/BRCosmetics/Athena/Items/Cosmetics/Characters/Character_DarkStance_Inferno]}
        }}]
      priority {curve {rowName=Tandem.Speech.Priority.OutfitReaction}}
     }
     {
      message="Aw, who's a good boy?? You, and you, and you..."
      contextRequirements [{requirement {
         class=FortControllerRequirement_HasCID
         properties {allowedCIds [/BRCosmetics/Athena/Items/Cosmetics/Characters/Character_GateHound_Spectral]}
        }}]
      priority {curve {rowName=Tandem.Speech.Priority.OutfitReaction}}
     }
     {
      message="New place, new me. Might take up whittling."
      contextRequirements [{requirement {class=FortControllerRequirement_CheckNPCMatchConversationCount, properties {nPCUniqueIdTag {tagName=AISpawnerData.Type.Tandem.CopperToll}}}}]
      priority {curve {rowName=None}}
     }
     {
      message="You don't know how nice it is to finally see the SUN, mate!"
      contextRequirements [{requirement {class=FortControllerRequirement_CheckNPCMatchConversationCount, properties {nPCUniqueIdTag {tagName=AISpawnerData.Type.Tandem.CopperToll}}}}]
      priority {curve {rowName=None}}
     }
     {
      message="I'm loving this Island, mate. Fresh air. FOOD? It's the dream."
      contextRequirements [{requirement {class=FortControllerRequirement_CheckNPCMatchConversationCount, properties {nPCUniqueIdTag {tagName=AISpawnerData.Type.Tandem.CopperToll}}}}]
      priority {curve {rowName=None}}
     }
     {
      message=Name's Charon. I used to run shades across the Styx to the Underworld.
      contextRequirements [{requirement {class=FortControllerRequirement_CheckNPCMatchConversationCount, properties {nPCUniqueIdTag {tagName=AISpawnerData.Type.Tandem.CopperToll}}}}]
      priority {curve {rowName=None}}
     }
     {
      message=You ever been to the Elysian Fields? Beautiful place. Very clean.
      contextRequirements [{requirement {class=FortControllerRequirement_CheckNPCMatchConversationCount, properties {nPCUniqueIdTag {tagName=AISpawnerData.Type.Tandem.CopperToll}}}}]
      priority {curve {rowName=None}}
     }
     {
      message="You ever been to the Asphodel Meadows? It's... a lot of ashes, mostly."
      contextRequirements [{requirement {class=FortControllerRequirement_CheckNPCMatchConversationCount, properties {nPCUniqueIdTag {tagName=AISpawnerData.Type.Tandem.CopperToll}}}}]
      priority {curve {rowName=None}}
     }
     {
      message="You been to Tartarus, mate? Fun for visitors, not for residents."
      contextRequirements [{requirement {class=FortControllerRequirement_CheckNPCMatchConversationCount, properties {nPCUniqueIdTag {tagName=AISpawnerData.Type.Tandem.CopperToll}}}}]
      priority {curve {rowName=None}}
     }
    ]
   }
  }
 }
 ConversationLinkNode_0 {id=#4, class=ConversationLinkNode, properties {remoteEntryTag {tagName=Conversation.Entry.BasicServices}}}
 ConversationLinkNode_1 {id=#3, class=ConversationLinkNode, properties {remoteEntryTag {tagName=Conversation.Entry.NPCLibrary.Quests}}}
]
